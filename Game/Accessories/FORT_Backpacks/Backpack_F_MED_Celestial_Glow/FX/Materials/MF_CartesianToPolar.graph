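# Material function that converts a 2D XY coordinate into polar Radius and Angle, stored with the Celestial Glow backpack's FX materials.
# XY defaults to texture coordinates remapped to a centered -1 to 1 range.
# Radius is the square root of the summed squares of X and Y.
# Angle uses fast arctangent2 divided by Tau, approximated as a constant ratio, then Frac wraps it to 0-1 for a full turn.
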MaterialFunction /Game/Accessories/FORT_Backpacks/Backpack_F_MED_Celestial_Glow/FX/Materials/MF_CartesianToPolar
graph {nodes [
  n1 = FunctionInput(Preview=ConstantBiasScale(None=TextureCoordinate()) {bias=-0.5, scale=2}) {bUsePreviewValueAsDefault=true, inputName=XY, inputType=FunctionInput_Vector2}
  n2 = ComponentMask(None=@n1) {g=true}
  n3 = Multiply(A=@n2, B=@n2)
  n4 = ComponentMask(None=@n1) {r=true}
  n5 = Multiply(A=@n4, B=@n4)
  FunctionOutput(None=SquareRoot(None=Add(A=@n3, B=@n5))) {outputName=Radius, sortPriority=1}
  n6 = Arctangent2Fast(Y=@n2, X=@n4)
  n7 = Divide() {constA=710, constB=113, desc=Tau}
  FunctionOutput(None=Frac(None=Divide(A=@n6, B=@n7))) {description=0-1 equals 0-360º, outputName=Angle}
 ]}
