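# Material function MF_PolarUV in the Celestial Glow backpack FX folder that converts UVs to polar coordinates with an adjustable twist.
# The input UV, TexCoord 0 by default, is remapped around the center and fed to MF_CartesianToPolar.
# Angle is offset by a radius-driven blend of center and edge twist, then paired with radius.
# DDX and DDY of the original UV are also output for texture samplers.
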MaterialFunction /Game/Accessories/FORT_Backpacks/Backpack_F_MED_Celestial_Glow/FX/Materials/MF_PolarUV
graph {nodes [
  n1 = FunctionInput(Preview=TextureCoordinate()) {bUsePreviewValueAsDefault=true, description="Defaults to TexCoord[0]", inputName=UV, inputType=FunctionInput_Vector2}
  n2 = MaterialFunctionCall!(XY=ConstantBiasScale(None=@n1) {bias=-0.5, scale=2}) {materialFunction=&/MF_CartesianToPolar}
  n3 = FunctionInput() {
   bUsePreviewValueAsDefault=true
   description=Twists the UVs from the center
   inputName=Twist Center
   inputType=FunctionInput_Scalar
   sortPriority=1
  }
  n4 = FunctionInput() {
   bUsePreviewValueAsDefault=true
   description=Twists the UVs from the edge
   inputName=Twist Edges
   inputType=FunctionInput_Scalar
   sortPriority=2
  }
  n5 = LinearInterpolate(A=@n3, B=@n4, Alpha=@n2.Radius)
  FunctionOutput(None=AppendVector(A=Add(A=@n2.Angle, B=@n5), B=@n2.Radius)) {description=Cartesian UVs transformed to Polar Coordinates, outputName=Polar UVs}
  FunctionOutput(None=DDX(Value=@n1)) {description=DDX to use in the texture sampler, outputName=DDX, sortPriority=1}
  FunctionOutput(None=DDY(Value=@n1)) {description=DDY to use in the texture sampler, outputName=DDY, sortPriority=2}
 ]}
